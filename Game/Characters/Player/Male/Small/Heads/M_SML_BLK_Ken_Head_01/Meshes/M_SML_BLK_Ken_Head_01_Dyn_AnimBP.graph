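# Animation Blueprint for the Ken small male player head on the Fortnite base head skeleton, passing its Input Pose straight to Output Pose.
# An unconnected chain of five Anim Dynamics nodes sets up low-gravity jiggle on the dyn_hair bones, with tight per-bone linear limits.
# That chain and a lone RigidBody node never reach the output, so neither affects the pose.
# The event graph holds only unwired stub nodes.
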AnimBlueprint /Game/Characters/Player/Male/Small/Heads/M_SML_BLK_Ken_Head_01/Meshes/M_SML_BLK_Ken_Head_01_Dyn_AnimBP
target_skeleton=/Game/Characters/Player/Common/Fortnite_Base_Head/Fortnite_Base_Head_Export_Skeleton
graphs [
 AnimGraph {nodes [
   n1 = Anim Dynamics - Bone: dyn_hair_03(ComponentPose=Local To Component(), Alpha=1, ExternalForce="0,0,0") {
    bGravityOverrideInSimSpace=true
    gravityScale=0.01
    alphaBoolBlend {blendOption=Linear}
    boundBone {boneName=dyn_hair_03}
    physicsBodyDefinitions [{
      sphereCollisionRadius=0
      boundBone {boneName=dyn_hair_03}
      boxExtents=<2 2 2>
      constraintSetup {angularLimitsMax=<0 0 0>, angularLimitsMin=<0 0 0>, angularTarget=<0 0 0>, linearAxesMax=<0.7 0.7 0.7>, linearAxesMin=<-0.7 -0.7 -0.7>}
      localJointOffset=<0 0 0>
     }]
   }
   n2 = Anim Dynamics - Bone: dyn_hair_04(ComponentPose=@n1, Alpha=1, ExternalForce="0,0,0") {
    bGravityOverrideInSimSpace=true
    gravityScale=0.01
    alphaBoolBlend {blendOption=Linear}
    boundBone {boneName=dyn_hair_04}
    physicsBodyDefinitions [{
      sphereCollisionRadius=0
      boundBone {boneName=dyn_hair_04}
      boxExtents=<2 2 2>
      constraintSetup {angularLimitsMax=<0 0 0>, angularLimitsMin=<0 0 0>, angularTarget=<0 0 0>, linearAxesMax=<0.6 0.6 0.6>, linearAxesMin=<-0.6 -0.6 -0.6>}
      localJointOffset=<0 0 0>
     }]
   }
   n3 = Anim Dynamics - Bone: dyn_hair_02(ComponentPose=@n2, Alpha=1) {
    bGravityOverrideInSimSpace=true
    gravityScale=0.01
    alphaBoolBlend {blendOption=Linear}
    boundBone {boneName=dyn_hair_02}
    physicsBodyDefinitions [{
      sphereCollisionRadius=0
      boundBone {boneName=dyn_hair_02}
      boxExtents=<2 2 2>
      constraintSetup {angularLimitsMax=<0 0 0>, angularLimitsMin=<0 0 0>, angularTarget=<0 0 0>, linearAxesMax=<0.5 0.5 0.5>, linearAxesMin=<-0.5 -0.5 -0.5>}
      localJointOffset=<0 0 0>
     }]
   }
   n4 = Anim Dynamics - Bone: dyn_hair_01(ComponentPose=@n3, Alpha=1) {
    bGravityOverrideInSimSpace=true
    gravityScale=0.01
    alphaBoolBlend {blendOption=Linear}
    boundBone {boneName=dyn_hair_01}
    physicsBodyDefinitions [{
      sphereCollisionRadius=0
      boundBone {boneName=dyn_hair_01}
      boxExtents=<2 2 2>
      constraintSetup {angularLimitsMax=<0 0 0>, angularLimitsMin=<0 0 0>, angularTarget=<0 0 0>, linearAxesMax=<0.4 0.4 0.4>, linearAxesMin=<-0.4 -0.4 -0.4>}
      localJointOffset=<0 0 0>
     }]
   }
   n5 = Anim Dynamics - Bone: dyn_hair_05(ComponentPose=@n4, Alpha=1, ExternalForce="0,0,0") {
    bGravityOverrideInSimSpace=true
    gravityScale=0.01
    alphaBoolBlend {blendOption=Linear}
    boundBone {boneName=dyn_hair_05}
    physicsBodyDefinitions [{
      sphereCollisionRadius=0
      boundBone {boneName=dyn_hair_05}
      boxExtents=<2 2 2>
      constraintSetup {angularLimitsMax=<0 0 0>, angularLimitsMin=<0 0 0>, angularTarget=<0 0 0>, linearAxesMax=<0.4 0.4 0.4>, linearAxesMin=<-0.4 -0.4 -0.4>}
      localJointOffset=<0 0 0>
     }]
   }
   Component To Local(ComponentPose=@n5)
   RigidBody() {alphaBoolBlend {blendOption=Linear}}
   AnimGraphNode_SubInput_0 = Input Pose()
   Output Pose(Result=@AnimGraphNode_SubInput_0) {layerGroup=None}
  ]}
 EventGraph {nodes [
   TryGetPawnOwner()
   Event BlueprintUpdateAnimation()
  ]}
]
variables []
